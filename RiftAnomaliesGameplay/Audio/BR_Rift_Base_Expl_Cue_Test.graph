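# Test Sound Cue for a Rift Anomaly base explosion, judging by its name, with an empty node graph and no root, so it references no audio.
# It carries playback setup only: the explosives sound class, reduced volume and a Rift Anomaly attenuation asset alongside stored attenuation values for natural falloff and low-pass filtering.
# Its concurrency override caps simultaneous instances, stopping the oldest and applying volume scaling.
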SoundCue /RiftAnomaliesGameplay/Audio/BR_Rift_Base_Expl_Cue_Test
root=null
nodes []
properties {
 attenuationSettings=&/ATT_BR_Rift_Anomaly_Attenuation
 bOverrideConcurrency=true
 maxDistance=1010000
 soundClassObject=/Game/Sounds/FortSoundClassesAndModes/Fort_Explosives
 volumeMultiplier=0.4987
 attenuationOverrides {distanceAlgorithm=NaturalSound, falloffDistance=3800, lPFRadiusMax=3000, lPFRadiusMin=2000, attenuationShapeExtents {x=2200}}
 concurrencyOverrides {maxCount=4, resolutionRule=StopOldest, volumeScale=0.5}
}
