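# Sound cue for the rift base explosion in the Rift Anomalies feature, played through the Fort_Events sound class.
# A mixer layers randomized final blasts immediately, a boosted single tail after about half a second, rift-close variations around two seconds and distant build-up variations around 3.6 seconds.
# Each layer carries its own long-range logarithmic attenuation override with occlusion and a distance-curved submix send.
# Concurrency is capped at three instances.
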SoundCue /RiftAnomaliesGameplay/Audio/SC_BR_Rift_Base_Expl_Cue
root=Mixer_0
nodes [
 {
  id=Attenuation_0
  children [Random_0]
  properties {
   bOverrideAttenuation=true
   attenuationOverrides {
    distanceAlgorithm=Logarithmic
    falloffDistance=150000
    occlusionLowPassFilterFrequency=18000
    occlusionTraceChannel=ECC_Camera
    occlusionVolumeAttenuation=0.9
    reverbDistanceMax=250000
    stereoSpread=5000
    submixSendSettings [{
      bEnableHPFCutoff=false
      bEnableLPFCutoff=false
      disableManualSendClamp=false
      hPFCutoff=20
      lPFCutoff=20000
      maxSendDistance=150000
      maxSendLevel=1
      minSendDistance=400
      minSendLevel=0
      sendLevel=0.2
      sendLevelControlMethod=CustomCurve
      soundSubmix=null
      customSendLevelCurve {editorCurveData {keys [<at 0.1282 1.2567>,<at 0.4601 0.9675>,<at 0.8707 0.5292>,<at 1.4629 0.0238>]}}
     }]
   }
  }
 }
 {
  id=Attenuation_1
  children [Delay_0]
  properties {
   bOverrideAttenuation=true
   attenuationOverrides {
    distanceAlgorithm=Logarithmic
    falloffDistance=150000
    occlusionLowPassFilterFrequency=18000
    occlusionTraceChannel=ECC_Camera
    occlusionVolumeAttenuation=0.9
    reverbDistanceMax=250000
    stereoSpread=5000
    submixSendSettings [{
      bEnableHPFCutoff=false
      bEnableLPFCutoff=false
      disableManualSendClamp=false
      hPFCutoff=20
      lPFCutoff=20000
      maxSendDistance=150000
      maxSendLevel=1
      minSendDistance=400
      minSendLevel=0
      sendLevel=0.2
      sendLevelControlMethod=CustomCurve
      soundSubmix=null
      customSendLevelCurve {editorCurveData {keys [<at 0.1283 1.2567>,<at 0.4601 0.9675>,<at 0.8708 0.5292>,<at 1.4629 0.0238>]}}
     }]
   }
  }
 }
 {
  id=Attenuation_2
  children [Delay_1]
  properties {
   bOverrideAttenuation=true
   attenuationOverrides {
    distanceAlgorithm=Logarithmic
    falloffDistance=150000
    occlusionLowPassFilterFrequency=18000
    occlusionTraceChannel=ECC_Camera
    occlusionVolumeAttenuation=0.9
    reverbDistanceMax=250000
    stereoSpread=10000
    submixSendSettings [{
      bEnableHPFCutoff=false
      bEnableLPFCutoff=false
      disableManualSendClamp=false
      hPFCutoff=20
      lPFCutoff=20000
      maxSendDistance=150000
      maxSendLevel=1
      minSendDistance=400
      minSendLevel=0
      sendLevel=0.2
      sendLevelControlMethod=CustomCurve
      soundSubmix=null
      customSendLevelCurve {editorCurveData {keys [<at 0.1283 1.2567>,<at 0.4601 0.9675>,<at 0.8708 0.5292>,<at 1.4629 0.0238>]}}
     }]
   }
  }
 }
 {
  id=Attenuation_3
  children [Delay_2]
  properties {
   bOverrideAttenuation=true
   attenuationOverrides {
    distanceAlgorithm=Logarithmic
    falloffDistance=150000
    occlusionLowPassFilterFrequency=18000
    occlusionTraceChannel=ECC_Camera
    occlusionVolumeAttenuation=0.9
    reverbDistanceMax=250000
    stereoSpread=10000
    submixSendSettings [{
      bEnableHPFCutoff=false
      bEnableLPFCutoff=false
      disableManualSendClamp=false
      hPFCutoff=20
      lPFCutoff=20000
      maxSendDistance=150000
      maxSendLevel=1
      minSendDistance=400
      minSendLevel=0
      sendLevel=0.2
      sendLevelControlMethod=CustomCurve
      soundSubmix=null
      customSendLevelCurve {editorCurveData {keys [<at 0.1283 1.2567>,<at 0.4601 0.9675>,<at 0.8708 0.5292>,<at 1.4629 0.0238>]}}
     }]
   }
  }
 }
 {id=Delay_0, children [Random_1], properties {delayMax=1.8802, delayMin=2.2061}}
 {id=Delay_1, children [Random_2], properties {delayMax=3.5784, delayMin=3.5949}}
 {id=Delay_2, children [Random_3], properties {delayMax=0.5387, delayMin=0.368}}
 {id=Mixer_0, children [Attenuation_0, Attenuation_1, Attenuation_2, Attenuation_3], properties {inputVolume [1,0.6975,1.854,4.0066]}}
 {id=Random_0, children [WavePlayer_1,WavePlayer_2,WavePlayer_3,WavePlayer_4], properties {weights [1,1,1,1]}}
 {id=Random_1, children [WavePlayer_0,WavePlayer_5,WavePlayer_6,WavePlayer_7], properties {weights [1,1,1,1]}}
 {id=Random_2, children [WavePlayer_8, WavePlayer_9, WavePlayer_10], properties {weights [1,1,1]}}
 {id=Random_3, children [WavePlayer_12], properties {weights [1]}}
 {id=WavePlayer_0, properties {soundWaveAssetPtr=&/BlastBerry_Rift_Close_01}}
 {id=WavePlayer_1, properties {soundWaveAssetPtr=&/BR_Rift_Base_Final_01}}
 {id=WavePlayer_10, properties {soundWaveAssetPtr=&/MMObj_Rift_Container_BuildUp_Distant03}}
 {id=WavePlayer_12, properties {soundWaveAssetPtr=&/Rift_Base_Tail_02}}
 {id=WavePlayer_2, properties {soundWaveAssetPtr=&/BR_Rift_Base_Final_02}}
 {id=WavePlayer_3, properties {soundWaveAssetPtr=&/BR_Rift_Base_Final_03}}
 {id=WavePlayer_4, properties {soundWaveAssetPtr=&/BR_Rift_Base_Final_04}}
 {id=WavePlayer_5, properties {soundWaveAssetPtr=&/BlastBerry_Rift_Close_02}}
 {id=WavePlayer_6, properties {soundWaveAssetPtr=&/BlastBerry_Rift_Close_03}}
 {id=WavePlayer_7, properties {soundWaveAssetPtr=&/BlastBerry_Rift_Close_04}}
 {id=WavePlayer_8, properties {soundWaveAssetPtr=&/MMObj_Rift_Container_BuildUp_Distant01}}
 {id=WavePlayer_9, properties {soundWaveAssetPtr=&/MMObj_Rift_Container_BuildUp_Distant02}}
]
properties {
 bOverrideConcurrency=true
 maxDistance=150400
 soundClassObject=&Fort_Events
 subtitlePriority=10001
 volumeMultiplier=0.89
 attenuationOverrides {bAttenuate=false, bSpatialize=false, falloffDistance=1000000, stereoSpread=99463.8359, attenuationShapeExtents {x=10000}}
 concurrencyOverrides {maxCount=3}
}
